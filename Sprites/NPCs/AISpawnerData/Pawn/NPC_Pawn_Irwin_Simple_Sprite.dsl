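; Simple wildlife-based Sprite NPC pawn that players can pick up, damage, or reveal by radar.
; On the server it stores its spawn as GoalLocation, then registers delayed nearby-player perception driving a timer for interaction and alert hooks (UpdateAlerted is empty).
; Pickup force-kills the pawn and hands tracking ownership to the equipped held-Sprite weapon.
; Reveal adds indicator, stencil and optional compass marking; other events manage music tags, water audio and damage/death effects.
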(event EventBeginPlay
  ;[1 1 1] Init
  (|CallParentFunction)
  (CallFunction|DestroyUnusedComponents)
  (Default|RegisterAudioTag true)
  (Networking|SwitchHasAuthority
    (:Authority
      (Utilities|FlowControl|Sequence
        (:then_0
          (CallFunction|SetSpawnLocationtoBlackboard))
        (:then_1
          (Utilities|FlowControl|Delay (ScalableFloat|GetValueatLevel (Variables|Default|GetHF_DelayUntilDetection)))
          (Query|BPRegisterListener (Variables|Default|GetFortAINearbyActorsPerception) (Utilities|Struct|MakeFortNearbyActorsPerceptionConfiguration (ScalableFloat|GetValueatLevel (Variables|Awareness|GetSpriteDetectionUpdateTime)) 3 0 (ScalableFloat|GetValueatLevel (Variables|Awareness|GetSpritePlayerAwarenessRad))) (EventDispatchers|CreateEvent :!event Custom|OnDetectedPlayerChangedEvent)))))))

(event Custom|DestroyUnusedComponents
  (bind _returnvalue (Actor|GetComponentbyClass "/Script/FortniteGame.FortFootstepComponent"))
  (bind _returnvalue_2 (Actor|GetComponentbyClass "/Script/FortniteGame.FortSurfaceTrackingComponent"))
  ;[1 1 1] Init
  (Components|DestroyComponent (Variables|Default|GetWildlife_FootDust))
  (Utilities|IsValid _returnvalue
    (:Is%20Valid
      (Components|DestroyComponent _returnvalue))
    (:Is%20Not%20Valid
      _))
  (Utilities|IsValid _returnvalue_2
    (:Is%20Valid
      (Components|DestroyComponent _returnvalue_2))))

(event Custom|SetSpawnLocationToBlackboard
  (bind _returnvalue (Pawn|GetController))
  (bind _returnvalue_2 (AI|GetBlackboard _returnvalue))
  ;[1 1 1] Init
  ;;[1 1 1] Spawn Location to BB
  (Utilities|IsValid _returnvalue
    (:Is%20Valid
      (Utilities|IsValid _returnvalue_2
        (:Is%20Valid
          (AI|Components|Blackboard|SetValueasVector _returnvalue_2 (Utilities|String|StringToName "GoalLocation") (Transformation|GetActorLocation)))))))

(event EventOnNPCPawnInteractedWith
  (params (AIPawn :type Actor%20Object%20Reference) (InteractingPlayerPawn :type Fort%20Player%20Pawn%20Object%20Reference) (InteractionBeingAttempted :type EInteractionBeingAttempted%20Enum :default "FirstInteraction"))
  (bind _returnvalue (Fortnite|GetFortPlayerControllerfromActor InteractingPlayerPawn))
  ;[1 1 1] Interact to cause slurp reaction
  (Default|RegisterAudioTag false)
  (Class|BPSpritesNPCMusicComponent|StopSpriteNPCMusic (Variables|Default|GetBP_Sprites_NPC_Music_Component))
  (bind _spritegiven (Class|BPSpritePickupComponent|PickupSprite (Variables|Default|GetBP_SpritePickupComponent) _returnvalue))
  (if _spritegiven
    (bind _output_get (Variables|Default|SetCurrentInteractingPawn InteractingPlayerPawn))
    (FortPawn|BindEventtoOnWeaponEquippedDelegate _output_get (EventDispatchers|CreateEvent :!event OnSpriteWeaponEquipped))
    (Default|CallOnSpritePickedUp _returnvalue)
    (Ability|SendGameplayEventtoActor InteractingPlayerPawn "(TagName=\"Gameplay.Action.Sprites.Interaction.PickedUp\")")
    (IndicatedManagement|RemoveActorFromIndicatedList _returnvalue self)
    (FortPawn|ForceKill)))

(event EventOnDamageServer
  (params (Damage :type Float%20%28single-precision%29) (DamageTags :type Gameplay%20Tag%20Container%20Structure%20%28by%20ref%29) (Momentum :type Vector :default "0, 0, 0") (HitInfo :type Hit%20Result%20Structure%20%28by%20ref%29) (InstigatedBy :type Controller%20Object%20Reference) (DamageCauser :type Actor%20Object%20Reference) (EffectContext :type Gameplay%20Effect%20Context%20Handle%20Structure))
  ;[1 0.007519 0] Catch damage for instigator since it doesn't work without skeletal mesh
  (if (> Damage 0)
    (Variables|Default|SetLastDamageCauserRef DamageCauser)))

(event EventOnDeathPlayEffects
  (params (Damage :type Float%20%28single-precision%29) (DamageTags :type Gameplay%20Tag%20Container%20Structure%20%28by%20ref%29) (Momentum :type Vector :default "0, 0, 0") (HitInfo :type Hit%20Result%20Structure%20%28by%20ref%29) (InstigatedBy :type Fort%20Pawn%20Object%20Reference) (DamageCauser :type Actor%20Object%20Reference) (EffectContext :type Gameplay%20Effect%20Context%20Handle%20Structure))
  (bind _headparticlesystem (Variables|Default|GetHeadParticleSystem))
  ;[1 1 1] Hide mesh because effect doesn't work
  (Development|SetHiddeninGame (Variables|Character|GetMesh) true false)
  (Components|Activation|Deactivate _headparticlesystem)
  (Components|DestroyComponent _headparticlesystem)
  (Audio|Components|Audio|FadeOut (Variables|Default|GetMovementAudioComponent) 0.35 0.0 "Linear" :!class /Script/Engine.AudioComponent)
  (Components|Activation|Deactivate (Variables|Default|GetIdle%20Forward%20FX))
  (Default|RegisterAudioTag false)
  (Class|BPSpritesNPCMusicComponent|StopSpriteNPCMusic (Variables|Default|GetBP_Sprites_NPC_Music_Component)))

(event EventOnDamagePlayEffects
  (params (Damage :type Float%20%28single-precision%29) (DamageTags :type Gameplay%20Tag%20Container%20Structure%20%28by%20ref%29) (Momentum :type Vector :default "0, 0, 0") (HitInfo :type Hit%20Result%20Structure%20%28by%20ref%29) (InstigatedBy :type Fort%20Pawn%20Object%20Reference) (DamageCauser :type Actor%20Object%20Reference) (EffectContext :type Gameplay%20Effect%20Context%20Handle%20Structure))
  ;[1 1 1] Play Damage Effects
  (|CallParentFunction Damage DamageTags Momentum HitInfo InstigatedBy DamageCauser EffectContext)
  ;;[1 1 1] Play Sound when actor takes damage
  (Default|PlaySoundLib self "(TagName=\"SoundLibrary.Irwin.Damaged\")"))

(event WaterBody|EventOnEnteredWaterBodyClient
  (params (WaterBody :type Fort%20Water%20Body%20Actor%20Object%20Reference) (bIsFirstBody :type Boolean))
  ;[1 1 1] Play water movement audio
  (bind _mesh (Variables|Character|GetMesh)
    (:then
      (bind _watermovementaudiocomponent (Variables|Default|GetWaterMovementAudioComponent)
        (:then
          (Audio|Components|Audio|FadeOut _watermovementaudiocomponent 0.3 0.0 "Linear" :!class /Script/Engine.AudioComponent))
        (:else
          _))
      (bind _returnvalue (Audio|SpawnSoundAttached (Variables|Default|GetWaterMovementLoopSound) _mesh))
      (Variables|Default|SetWaterMovementAudioComponent _returnvalue))
    (:else)))

(event EventOnLanded
  (params (Hit :type Hit%20Result%20Structure%20%28by%20ref%29))
  (|CallParentFunction Hit)
  (Components|Movement|StopMovementImmediately (Variables|Character|GetCharacterMovement)))

(event WaterBody|EventOnExitedWaterBodyClient
  (params (WaterBody :type Fort%20Water%20Body%20Actor%20Object%20Reference) (bIsLastBody :type Boolean))
  ;[1 1 1] Play water movement audio
  (bind _watermovementaudiocomponent (Variables|Default|GetWaterMovementAudioComponent)
    (:then
      (Audio|Components|Audio|FadeOut _watermovementaudiocomponent 0.3 0.0 "Linear" :!class /Script/Engine.AudioComponent))
    (:else)))

(event Custom|PlayRadarSound
  (params (Instigator :type Actor%20Object%20Reference))
  ;[1 1 1] Local player sound if shrine radar marks the sprite
  (Networking|CanExecuteCosmeticEvents
    (:True
      (bind _asfort_player_pawn (Utilities|Casting|CastToFortPlayerPawn Instigator))
      (if (Pawn|IsLocallyViewed _asfort_player_pawn)
        (Audio|PlaySoundatLocation (Variables|Default|GetMarkedSpriteSound) (Transformation|GetActorLocation))))))

(event Custom|OnDetectedPlayerChangedEvent
  (params (AllFoundActors :type Array%20of%20Actor%20Object%20References) (EnteringActors :type Array%20of%20Actor%20Object%20References) (ExitingActors :type Array%20of%20Actor%20Object%20References))
  (Utilities|Array|Clear (Variables|Awareness|GetPlayerPawnArray))
  (for _array_element AllFoundActors
    (bind _asfort_player_pawn (Utilities|Casting|CastToFortPlayerPawn _array_element))
    (Utilities|Array|Add (Variables|Awareness|GetPlayerPawnArray) _asfort_player_pawn))
  (Utilities|FlowControl|Sequence
    (:then_0
      (if (> (Utilities|Array|Length (Variables|Awareness|GetPlayerPawnArray)) 0)
        (CallFunction|HandleInteracts)
        (CallFunction|UpdateAlerted)
        (bind _returnvalue (Utilities|Time|SetTimerbyEvent (EventDispatchers|CreateEvent :!event Custom|OnDetectionUpdate) (ScalableFloat|GetValueatLevel (Variables|Awareness|GetSpriteDetectionUpdateTime)) true))
        (Variables|Awareness|SetDetectionUpdateTimerHandle _returnvalue)
        (else
          (Utilities|Time|ClearandInvalidateTimerbyHandle (Variables|Awareness|GetDetectionUpdateTimerHandle)))))
    (:then_1
      (if (Variables|Debug|GetbDebugPawn)
        (for _array_element_2 EnteringActors
          (Development|PrintString (Utilities|String|Append "Entering Pawn: " (Utilities|GetDisplayName _array_element_2))))
        (for _array_element_3 ExitingActors
          (Development|PrintString (Utilities|String|Append "Exiting Pawn: " (Utilities|GetDisplayName _array_element_3))))))))

(event Custom|OnDetectionUpdate
  (if (Variables|Debug|GetbDebugPawn)
    (Rendering|Debug|DrawDebugSphere (Transformation|GetActorLocation) (ScalableFloat|GetValueatLevel (Variables|Awareness|GetSpritePlayerAwarenessRad)) 12 "(R=0.000000,G=0.677083,B=0.481644,A=1.000000)" (ScalableFloat|GetValueatLevel (Variables|Awareness|GetSpriteDetectionUpdateTime)) 1.0))
  (CallFunction|HandleInteracts)
  (CallFunction|UpdateAlerted))

(event Custom|SaveMarkerDataClient
  (params (MarkerComponent :type Athena%20Marker%20Component%20Object%20Reference) (MarkerID :type Marker%20ID%20Structure))
  (Utilities|Map|Add (Variables|Default|GetActiveMarkingsMap) MarkerComponent MarkerID))

(fn BlueprintCanInteract
  (params (InteractingPawn :type Fort%20Pawn%20Object%20Reference) (InteractionBeingAttempted :type EInteractionBeingAttempted%20Enum :default "FirstInteraction"))
  (bind _bp_spritepickupcomponent (Variables|Default|GetBP_SpritePickupComponent))
  (bind _bcanpickup (Class|BPSpritePickupComponent|CanPickup _bp_spritepickupcomponent InteractingPawn))
  (bind _returnvalue (|CallParentFunction InteractingPawn InteractionBeingAttempted))
  (bind _returnvalue_2 (and _bcanpickup _returnvalue))
  (Development|PrintString (Utilities|String|Append "CAN INTERACT: " (Utilities|String|ToString(Boolean) _returnvalue_2)))
  (return _returnvalue_2))

(fn GetLaunchTrajectory
  (return (+ (* (Transformation|GetActorForwardVector (Variables|Default|GetpawnToLaunch)) (Math|Float|MapRangeClamped (Math|Vector|VectorLength (Math|Vector|MakeVector (.x (Transformation|GetVelocity (Variables|Default|GetpawnToLaunch))) (.y (Transformation|GetVelocity (Variables|Default|GetpawnToLaunch))))) (ScalableFloat|GetValueatLevel (Variables|Default|GetHF_Bounce_XYInputRangeMin)) (ScalableFloat|GetValueatLevel (Variables|Default|GetHF_Bounce_XYInputRangeMax)) (ScalableFloat|GetValueatLevel (Variables|Default|GetHF_Bounce_XYOutputRangeMin)) (ScalableFloat|GetValueatLevel (Variables|Default|GetHF_Bounce_XYOutputRangeMax)))) (Math|Vector|MakeVector 0.0 0.0 (ScalableFloat|GetValueatLevel (Variables|Default|GetHF_Bounce_ZVeloicty)))) false))

(fn NormalizeVelocity
  (return (Math|Float|NormalizetoRange (Math|Vector|VectorLength (Transformation|GetVelocity)) 2.5e+02)))

(fn NPCCanBeInteractedWith
  (return true))

(fn OnRep_PrimitiveDataFloatRandomizer
  (Rendering|Material|SetCustomPrimitiveDataFloat (Variables|Character|GetMesh) 0 (Variables|Default|GetPrimitiveDataFloatRandomizer)))

(fn OnRep_SkelMeshScale
  (Networking|CanExecuteCosmeticEvents
    (:True
      (Transformation|SetWorldScale3D (Variables|Character|GetMesh) (Math|Conversions|ToVector(Float) (Variables|Default|GetSkelMeshScale))))))

(fn OnSpriteWeaponEquipped
  (params (NewWeapon :type Fort%20Weapon%20Object%20Reference) (PrevWeapon :type Fort%20Weapon%20Object%20Reference))
  (bind _asb_sprite_held_weapon_parent (Utilities|Casting|CastToB_Sprite_Held_Weapon_Parent NewWeapon))
  (FortPawn|UnbindEventfromOnWeaponEquippedDelegate (Variables|Default|GetCurrentInteractingPawn) (EventDispatchers|CreateEvent :!event OnSpriteWeaponEquipped))
  (Class|BPSpriteTrackingComponent|TransferTrackingOwnership (Variables|Default|GetBP_SpriteTrackingComponent) (Actor|GetComponentbyClass _asb_sprite_held_weapon_parent "/Sprites/Gameplay/Components/BP_SpriteTrackingComponent.BP_SpriteTrackingComponent_C")))

(fn RegisterAudioTag
  (params (bRegister :type Boolean))
  (bind _self self)
  (bind _musictags (Variables|Music|GetMusicTags))
  ;[1 1 1] Handles adding required music tag to instigator actor
  (if bRegister
    (Ability|GameplayEffect|AddLooseGameplayTags _self _musictags true)
    (else
      (Ability|GameplayEffect|RemoveLooseGameplayTags _self _musictags true))))

(fn RevealSpriteOnMinimap
  (params (Instigator :type Actor%20Object%20Reference))
  (bind _returnvalue (Fortnite|GetFortPlayerControllerfromActor Instigator))
  (bind _self self)
  (bind _array (Utilities|Array|MakeArray _self))
  (Networking|SwitchHasAuthority
    (:Authority
      (IndicatedManagement|AddActorsToIndicatedList _returnvalue _array (IndicatedManagement|MakeIndicatedActorData :GroupIdentifier "SpriteRevealIndicatorGroup" :ScaleAndOpacityData (IndicatedManagement|MakeIndicatedActorScaleandOpacityData 7.5e+03 1e+01 1.0 2.0 0.8) :ParticleSystemData (IndicatedManagement|MakeIndicatedActorParticleSystemData (Variables|Indicators|GetIndicatedParticleSystem) (Variables|Indicators|GetIndicatedParticleSystemParamName) (Variables|Indicators|GetPS_Offset) (Variables|Indicators|GetPS_DBNO_Offset) (Variables|Indicators|GetPS_VectorParamName)) :Duration (ScalableFloat|GetValueatLevel (Variables|UI|GetHF_RevealDuration)) :StepTime 0.1 :IndicatorOffset "0.000000,0.000000,40.000000" :IndicatorDBNOOffset "0.000000,0.000000,0.000000" :ShareActorWith "AllTeam" :StateImageOverride "None" :StateIndexOverride (Variables|Indicators|GetIndicatorIndexOverride) :bClampToScreen true :bHideStateImage false :bOnlyDisplayWhenStencilingIsDisabled false :bOnlyDisplayIfActorIsVisible false) true false true)
      (CallFunction|PlayRadarSound Instigator)
      (IndicatedManagement|AddActorsToStenciledList _returnvalue _array (Variables|Default|GetStenciled%20Actor%20Data) true true)
      (if (ScalableFloat|GetValueasBool (Variables|UI|GetHF_RevealOnCompass))
        (Class|AthenaLocalCompassComponent|AddtoPlayerCompassforDuration (Variables|Default|GetAthenaLocalCompass) _returnvalue (ScalableFloat|GetValueatLevel (Variables|UI|GetHF_RevealDuration)))
        (else
          _))
      (Variables|Default|SetIsMarked? true))))

(fn TestAndSetMaxVelocity
  (bind _returnvalue (CallFunction|NormalizeVelocity))
  (if (> _returnvalue (Variables|Default|GetMaxNormalizedVelocity))
    (Variables|Default|SetMaxNormalizedVelocity _returnvalue)))

(fn TestIfRidingAndApplyFallImmunity
  (params (Object :type Object%20Reference))
  (bind _asfort_player_pawn (Utilities|Casting|CastToFortPlayerPawn Object))
  (if (GameplayTags|HasMatchingGameplayTag _asfort_player_pawn "(TagName=\"Riding.IsRidingActor\")")
    (bind _asnpc_pawn_wildlife_parent (Utilities|Casting|CastToNPC_Pawn_Wildlife_Parent (Components|GetOwner (Transformation|GetAttachParent (Variables|/Script/Engine.Character|GetCapsuleComponent _asfort_player_pawn)))))
    (Environment|ApplyGlobalEnvironmentGameplayEffecttoActor _asnpc_pawn_wildlife_parent (Variables|Default|GetGE_FallImmunity) 0 "(GameplayTags=)")))

(fn UnmarkSprite
  (bind _activemarkingsmap (Variables|Default|GetActiveMarkingsMap))
  (if (Variables|Default|GetIsMarked?)
    (Networking|CanExecuteCosmeticEvents
      (:True
        (bind _keys (Utilities|Map|Keys _activemarkingsmap))
        (for _array_element _keys
          (MarkerComponent|UnmarkActoronClient _array_element (Utilities|Map|Find _activemarkingsmap _array_element)))
        (Utilities|Map|Clear (Variables|Default|GetActiveMarkingsMap))
        (Variables|Default|SetIsMarked? false)))))

(fn UpdateAlerted)

(fn ConstructionScript
  (|CallParentFunction))
